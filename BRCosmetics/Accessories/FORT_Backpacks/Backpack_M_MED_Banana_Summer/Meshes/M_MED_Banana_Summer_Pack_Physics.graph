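# Physics asset for the Banana Summer back bling cosmetic, defining collision and joint limits for its dangling flaps, cup, zippers, loops and handle.
# Simple box and capsule bodies on dyn_ bones attach directly or in chains to a two-shape pack body.
# Flaps and zippers form single-axis hinges with swing2 and twist locked, with front flaps branching at the tips.
# Loops and cup get narrow limits, while the top handle rotates freely.
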PhysicsAsset /BRCosmetics/Accessories/FORT_Backpacks/Backpack_M_MED_Banana_Summer/Meshes/M_MED_Banana_Summer_Pack_Physics
bodies {
 dyn_backFlap_01_l [{shapeType=Box, extentX=1.1977, extentY=7.6339, extentZ=7.2139, center=<3.5411 0.0001 0>, rotation=<rot 0 90 0>}]
 dyn_backFlap_01_r [{shapeType=Box, extentX=1.1977, extentY=11.4266, extentZ=12.7562, center=<5.6093 0.0001 2.303>, rotation=<rot 0 90 0>}]
 dyn_backFlap_02_l [{shapeType=Box, extentX=0.8868, extentY=7.7842, extentZ=6.6506, center=<3.7224 -0.0002 0>, rotation=<rot 0 90 0>}]
 dyn_backFlap_02_r [{shapeType=Box, extentX=0.8868, extentY=7.7842, extentZ=14.54, center=<3.7224 -0.0003 1.8776>, rotation=<rot 0 90 0>}]
 dyn_backFlap_03_l [{shapeType=Box, extentX=1.1846, extentY=7.2673, extentZ=9.0291, center=<2.6365 0 0>, rotation=<rot 0 90 0>}]
 dyn_backFlap_03_r [{shapeType=Box, extentX=1.1846, extentY=6.151, extentZ=7.8317, center=<2.6365 0 0>, rotation=<rot 0 90 0>}]
 dyn_cup [{shapeType=Capsule, length=8.6139, radius=4.9599, center=<6.3839 2.198 -0.0599>, rotation=<rot -1.7589 100.4198 -90.1942>}]
 dyn_frontFlap_01_l [{shapeType=Box, extentX=1.1977, extentY=4.7448, extentZ=12.7562, center=<1.7064 0.0001 0>, rotation=<rot 0 90 0>}]
 dyn_frontFlap_01_r [{shapeType=Box, extentX=1.1977, extentY=11.4266, extentZ=12.7562, center=<5.6093 0.0001 0>, rotation=<rot 0 90 0>}]
 dyn_frontFlap_02_l [{shapeType=Box, extentX=0.8868, extentY=7.7842, extentZ=12.4545, center=<3.7224 -0.0002 0>, rotation=<rot 0 90 0>}]
 dyn_frontFlap_02_r [{shapeType=Box, extentX=2.1105, extentY=7.7842, extentZ=21.5791, center=<3.7224 0.6437 0.1019>, rotation=<rot 1.1025 89.9997 0>}]
 dyn_frontFlap_03_l [{shapeType=Box, extentX=1.1846, extentY=6.151, extentZ=11.3123, center=<2.6365 0 0>, rotation=<rot 0 90 0>}]
 dyn_frontFlap_03_r [{shapeType=Box, extentX=1.1846, extentY=6.151, extentZ=7.8317, center=<2.6365 0 0>, rotation=<rot 0 90 0>}]
 dyn_frontFlap_04a_l [{shapeType=Box, extentX=1.1846, extentY=6.151, extentZ=7.8317, center=<2.6365 0 0>, rotation=<rot 0 90 0>}]
 dyn_frontFlap_04a_r [{shapeType=Box, extentX=1.1846, extentY=6.151, extentZ=7.8317, center=<2.6365 0 0>, rotation=<rot 0 90 0>}]
 dyn_frontFlap_04b_l [{shapeType=Box, extentX=1.1846, extentY=6.151, extentZ=7.8317, center=<2.6365 0 0>, rotation=<rot 0 90 0>}]
 dyn_frontFlap_04b_r [{shapeType=Box, extentX=1.1846, extentY=6.151, extentZ=7.8317, center=<2.6365 0 0>, rotation=<rot 0 90 0>}]
 dyn_frontFlap_04c_l [{shapeType=Box, extentX=1.1846, extentY=6.151, extentZ=7.8317, center=<2.6365 0 0>, rotation=<rot 0 90 0>}]
 dyn_lowFlapFront [{shapeType=Box, extentX=1.5866, extentY=9.0662, extentZ=15.5202, center=<4.5 0 0>, rotation=<rot 0 90 0>}]
 dyn_lowFlap_l [{shapeType=Box, extentX=2.0111, extentY=8.8147, extentZ=17.5382, center=<4.5 0 0>, rotation=<rot 0 90 0>}]
 dyn_lowFlap_r [{shapeType=Box, extentX=3.5322, extentY=10.6529, extentZ=20.74, center=<4.5 0 0>, rotation=<rot 0 90 0>}]
 dyn_topHandle [{shapeType=Box, extentX=5.5203, extentY=3.3308, extentZ=11.8723, center=<0.4385 -2.4211 0.2997>, rotation=<rot -2.3941 -90 -0.2848>}]
 dyn_topLoop_01 [{shapeType=Box, extentX=2.29, extentY=6.5, extentZ=14.4, center=<4 0 0>, rotation=<rot -1.39 -90 0.47>}]
 dyn_topLoop_02 [{shapeType=Box, extentX=3, extentY=8.2, extentZ=13, center=<3.8 -1.38 -0.77>, rotation=<rot -3.63 -90 6.6>}]
 dyn_zipperBack [{shapeType=Box, extentX=0.7615, extentY=3.3103, extentZ=9.6871, center=<4.5382 -0.0242 0.3087>, rotation=<rot 0 89.4864 -79.6248>}]
 dyn_zipperFront [{shapeType=Box, extentX=0.6814, extentY=3.3163, extentZ=9.6871, center=<3.8447 -0.0152 0.3118>, rotation=<rot 0 89.6053 -79.3029>}]
 pack [
  {shapeType=Capsule, length=10.6229, radius=8.2092, center=<-3.8144 -13.0038 -26.1046>, rotation=<rot 78.1266 -57.1141 -0.0001>}
  {shapeType=Box, extentX=28.0445, extentY=11.2488, extentZ=8.9288, center=<1.4608 -5.628 5.0797>, rotation=<rot 0 0 21.691>}
 ]
}
constraints [
 {bone1Name=dyn_topLoop_02, bone2Name=dyn_topLoop_01, swing1LimitDegrees=20, swing2LimitDegrees=20, twistLimitDegrees=0}
 {bone1Name=dyn_topLoop_01, bone2Name=pack, swing1LimitDegrees=20, swing2LimitDegrees=20, twistLimitDegrees=0}
 {
  bone1Name=dyn_zipperFront
  bone2Name=pack
  swing1LimitDegrees=45
  swing2LimitDegrees=45
  swing2Motion=Locked
  twistLimitDegrees=45
  twistMotion=Locked
 }
 {
  bone1Name=dyn_zipperBack
  bone2Name=pack
  swing1LimitDegrees=45
  swing2LimitDegrees=45
  swing2Motion=Locked
  twistLimitDegrees=45
  twistMotion=Locked
 }
 {
  bone1Name=dyn_lowFlap_l
  bone2Name=pack
  swing1LimitDegrees=18
  swing2LimitDegrees=45
  swing2Motion=Locked
  twistLimitDegrees=45
  twistMotion=Locked
 }
 {
  bone1Name=dyn_lowFlap_r
  bone2Name=pack
  swing1LimitDegrees=18
  swing2LimitDegrees=45
  swing2Motion=Locked
  twistLimitDegrees=45
  twistMotion=Locked
 }
 {
  bone1Name=dyn_lowFlapFront
  bone2Name=pack
  swing1LimitDegrees=18
  swing2LimitDegrees=45
  swing2Motion=Locked
  twistLimitDegrees=45
  twistMotion=Locked
 }
 {
  bone1Name=dyn_topHandle
  bone2Name=pack
  swing1LimitDegrees=45
  swing1Motion=Free
  swing2LimitDegrees=45
  swing2Motion=Free
  twistLimitDegrees=45
  twistMotion=Free
 }
 {
  bone1Name=dyn_backFlap_01_r
  bone2Name=pack
  swing1LimitDegrees=25
  swing2LimitDegrees=45
  swing2Motion=Locked
  twistLimitDegrees=45
  twistMotion=Locked
 }
 {
  bone1Name=dyn_backFlap_02_r
  bone2Name=dyn_backFlap_01_r
  swing1LimitDegrees=20
  swing2LimitDegrees=45
  swing2Motion=Locked
  twistLimitDegrees=45
  twistMotion=Locked
 }
 {
  bone1Name=dyn_backFlap_03_r
  bone2Name=dyn_backFlap_02_r
  swing1LimitDegrees=25
  swing2LimitDegrees=45
  swing2Motion=Locked
  twistLimitDegrees=45
  twistMotion=Locked
 }
 {
  bone1Name=dyn_backFlap_01_l
  bone2Name=pack
  swing1LimitDegrees=25
  swing2LimitDegrees=45
  swing2Motion=Locked
  twistLimitDegrees=45
  twistMotion=Locked
 }
 {
  bone1Name=dyn_backFlap_02_l
  bone2Name=dyn_backFlap_01_l
  swing1LimitDegrees=25
  swing2LimitDegrees=45
  swing2Motion=Locked
  twistLimitDegrees=45
  twistMotion=Locked
 }
 {
  bone1Name=dyn_backFlap_03_l
  bone2Name=dyn_backFlap_02_l
  swing1LimitDegrees=25
  swing2LimitDegrees=45
  swing2Motion=Locked
  twistLimitDegrees=45
  twistMotion=Locked
 }
 {bone1Name=dyn_cup, bone2Name=pack, swing1LimitDegrees=20, swing2LimitDegrees=12, twistLimitDegrees=6}
 {
  bone1Name=dyn_frontFlap_01_r
  bone2Name=pack
  swing1LimitDegrees=10
  swing2LimitDegrees=45
  swing2Motion=Locked
  twistLimitDegrees=45
  twistMotion=Locked
 }
 {
  bone1Name=dyn_frontFlap_02_r
  bone2Name=dyn_frontFlap_01_r
  swing1LimitDegrees=15
  swing2LimitDegrees=45
  swing2Motion=Locked
  twistLimitDegrees=45
  twistMotion=Locked
 }
 {
  bone1Name=dyn_frontFlap_03_r
  bone2Name=dyn_frontFlap_02_r
  swing1LimitDegrees=20
  swing2LimitDegrees=45
  swing2Motion=Locked
  twistLimitDegrees=45
  twistMotion=Locked
 }
 {
  bone1Name=dyn_frontFlap_04a_r
  bone2Name=dyn_frontFlap_03_r
  swing1LimitDegrees=25
  swing2LimitDegrees=45
  swing2Motion=Locked
  twistLimitDegrees=45
  twistMotion=Locked
 }
 {
  bone1Name=dyn_frontFlap_04b_r
  bone2Name=dyn_frontFlap_03_r
  swing1LimitDegrees=25
  swing2LimitDegrees=45
  swing2Motion=Locked
  twistLimitDegrees=45
  twistMotion=Locked
 }
 {
  bone1Name=dyn_frontFlap_01_l
  bone2Name=pack
  swing1LimitDegrees=25
  swing2LimitDegrees=45
  swing2Motion=Locked
  twistLimitDegrees=45
  twistMotion=Locked
 }
 {
  bone1Name=dyn_frontFlap_02_l
  bone2Name=dyn_frontFlap_01_l
  swing1LimitDegrees=25
  swing2LimitDegrees=45
  swing2Motion=Locked
  twistLimitDegrees=45
  twistMotion=Locked
 }
 {
  bone1Name=dyn_frontFlap_03_l
  bone2Name=dyn_frontFlap_02_l
  swing1LimitDegrees=25
  swing2LimitDegrees=45
  swing2Motion=Locked
  twistLimitDegrees=45
  twistMotion=Locked
 }
 {
  bone1Name=dyn_frontFlap_04a_l
  bone2Name=dyn_frontFlap_03_l
  swing1LimitDegrees=25
  swing2LimitDegrees=45
  swing2Motion=Locked
  twistLimitDegrees=45
  twistMotion=Locked
 }
 {
  bone1Name=dyn_frontFlap_04b_l
  bone2Name=dyn_frontFlap_03_l
  swing1LimitDegrees=25
  swing2LimitDegrees=45
  swing2Motion=Locked
  twistLimitDegrees=45
  twistMotion=Locked
 }
 {
  bone1Name=dyn_frontFlap_04c_l
  bone2Name=dyn_frontFlap_03_l
  swing1LimitDegrees=25
  swing2LimitDegrees=45
  swing2Motion=Locked
  twistLimitDegrees=45
  twistMotion=Locked
 }
]
